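# Level sequence for the Pilgrim camera manager's battle-stage attack shot.
# A spawned Cine Camera Actor holds a fixed transform behind and above its origin, pitched slightly down, with constant wide focal length, focus distance and aperture.
# A camera shake section on the camera component adds a brief jolt early in the shot, while a single camera cut section spans the sequence.
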LevelSequence /PilgrimCore/CameraManager/Sequences/BattleStage/AttackShots/LS_PilgrimCamera_Attack
bindings [
 CameraComponent {
  id=#1
  parent_id=#2
  possessed_class=CineCameraComponent
  tracks [
   {
    type=MovieSceneFloatTrack
    display_name=Current Focal Length
    sections [{type=MovieSceneFloatSection, channels [None {type=MovieSceneScriptingFloatChannel, default=14.158, const [14.158,1]}]}]
   }
   {
    type=MovieSceneFloatTrack
    display_name="Manual Focus Distance (Focus Settings)"
    sections [{type=MovieSceneFloatSection, channels [None {type=MovieSceneScriptingFloatChannel, default=100000}]}]
   }
   {
    type=MovieSceneFloatTrack
    display_name=Current Aperture
    sections [{type=MovieSceneFloatSection, channels [None {type=MovieSceneScriptingFloatChannel, default=4.2382}]}]
   }
   {type=MovieSceneCameraShakeTrack, sections [{type=MovieSceneCameraShakeSection, end_frame=92, start_frame=50}]}
  ]
 }
 Cine Camera Actor {
  id=#2
  possessed_class=CineCameraActor
  tracks [
   {type=MovieSceneSpawnTrack, sections [{type=MovieSceneSpawnSection, channels [None {type=MovieSceneScriptingBoolChannel, default=true}]}]}
   {
    type=MovieScene3DTransformTrack
    sections [{
      type=MovieScene3DTransformSection
      channels [
       Location.X {type=MovieSceneScriptingDoubleChannel, default=-274.4, const [-274.4,1]}
       Location.Y {type=MovieSceneScriptingDoubleChannel, default=55.75, const [55.75,1]}
       Location.Z {type=MovieSceneScriptingDoubleChannel, default=110, const [110,1]}
       Rotation.X {type=MovieSceneScriptingDoubleChannel, default=0, const [0,1]}
       Rotation.Y {type=MovieSceneScriptingDoubleChannel, default=-11.5, const [-11.5,1]}
       Rotation.Z {type=MovieSceneScriptingDoubleChannel, default=0, const [0,1]}
       Scale.X {type=MovieSceneScriptingDoubleChannel, default=1, const [1,1]}
       Scale.Y {type=MovieSceneScriptingDoubleChannel, default=1, const [1,1]}
       Scale.Z {type=MovieSceneScriptingDoubleChannel, default=1, const [1,1]}
      ]
     }]
   }
  ]
 }
]
display_rate {denominator=1, numerator=30}
tick_resolution {denominator=1, numerator=24000}
tracks [{type=MovieSceneCameraCutTrack, sections [{type=MovieSceneCameraCutSection, end_frame=602, start_frame=0}]}]
